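# Subsurface material for an amber resin ingredient in the Creative STW set, with a backpack-shrinking effect built from LODEffect and BlackHoleFunction vertex-offset math.
# Tangent-space camera vector, AO and distortion normals create parallax UV offsets for bubble and noise layers, faking internal depth.
# These layers and Fresnel blend two resin colors, tinting the diffuse texture and feeding subsurface and specular math.
# AO masks Hologram emissive; a static switch adds the resin normal map.
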Material /Game/Creative/Sets/STW/Ingredients/Materials/CP_BackPackShrinkingEffect_Amber
graph {
 properties {pixelDepthOffsetMode=PDOM_Legacy, shadingModel=MSM_Subsurface, usedShadingModels=Subsurface}
 nodes [
  n1 = VectorParameter() {parameterName=WorldPostionOfBlackHole, defaultValue=<rgba -30000 25536 1952 0>}
  n2 = MaterialFunctionCall!(
   World Position=WorldPosition().XYZ
   DistanceCullFadeAnimationOffset=MaterialFunctionCall() {materialFunction=/Game/Packages/Fortress_SharedMaterials/Base_Material_Functions/PickupEffectDistanceCullOverride}.Result
   Use actor center instead of pivot point
  ) {materialFunction=/Game/Packages/Fortress_SharedMaterials/Base_Material_Functions/LODEffect}
  n3 = WorldPosition()
  Add(
   A=MaterialFunctionCall!(WorldPostionOfBlackHole=AppendVector(A=@n1.RGB, B=@n1.RGB), World Position=Add(A=@n2.Result, B=@n3.XYZ)) {materialFunction=/Game/Packages/Fortnite_UniqueMaterials_1/BlackHoleFunction}.Result
   B=@n2.Result
  )
  Add()
  n4 = TextureSample() {texture=/Game/Creative/Sets/STW/Ingredients/Textures/CP_T_Resin_AO}
  n5 = Multiply(A=Transform(None=CameraVectorWS()) {transformSourceType=TRANSFORMSOURCE_World, transformType=TRANSFORM_Tangent}, B=@n4.R)
  n6 = TextureSampleParameter2D!(
   UVs=Multiply(A=TextureCoordinate(), B=ScalarParameter() {defaultValue=3, parameterName=Amount of Distortion Map Tiling})
   Level
   Apply View MipBias
  ) {
   mipValueMode=TMVM_MipLevel
   parameterName=DistortionNormal
   samplerType=SAMPLERTYPE_Normal
   texture=/Game/Global/Textures/Default/Noise/T_Normal_Noise01
  }
  n7 = ComponentMask(None=Multiply(A=Multiply(A=@n5, B=@n6.RGB), B=ScalarParameter() {defaultValue=0.25, parameterName=Bump offset}) {constB=0.25}) {g=true, r=true}
  n8 = TextureSample(UVs=Add(A=Multiply(A=@n7) {constB=-3}, B=TextureCoordinate() {uTiling=15, vTiling=15})) {samplerType=SAMPLERTYPE_Normal, texture=/Game/Items/Art_noLOD/Textures/Bubbles}
  n9 = Multiply(A=ComponentMask(None=@n8.RGB) {g=true, r=true}) {constB=0.2}
  n10 = TextureCoordinate()
  n11 = ScalarParameter() {defaultValue=3, parameterName=Amount of color transition texture tiling}
  n12 = TextureSampleParameter2D!(UVs=Add(A=Add(A=@n7, B=@n9), B=Multiply(A=@n10, B=@n11)), Apply View MipBias) {parameterName=ColorTransitionImage, samplerType=SAMPLERTYPE_LinearColor, texture=/Game/Global/Textures/Default/Noise/T_Noise01_Inverse}
  n13 = PixelNormalWS()
  n14 = Constant3Vector() {constant=<rgba 1 1 0 0>}
  n15 = Multiply(A=@n8.B, B=Abs(None=DotProduct(A=@n8.RGB, B=@n14)))
  n16 = Add(A=Multiply(A=Multiply(A=@n4.R, B=@n12.R), B=Fresnel(Normal=@n13) {baseReflectFraction=0, exponent=1.5}), B=Multiply(A=@n15, B=@n15))
  n17 = Multiply(
   A=LinearInterpolate(
    A=VectorParameter() {parameterName=Res Col 2, defaultValue=<rgba 0.5664 0.2679 0.0412 1>}.RGB
    B=VectorParameter() {parameterName=Res Col 1, defaultValue=<rgba 1 0.9415 0 1>}.RGB
    Alpha=Clamp(None=@n16)
   )
   B=@n12.R
  )
  n18 = Multiply(A=Multiply(A=@n17, B=@n17), B=TextureSample() {texture=/Game/Creative/Sets/STW/Ingredients/Textures/CP_T_Resin_D}.RGB)
  n19 = ScalarParameter() {defaultValue=0.125, parameterName=SSS Desaturation}
  n20 = Fresnel(Normal=PixelNormalWS()) {baseReflectFraction=0, exponent=2}
  LinearInterpolate(
   A=Multiply(A=Desaturation(None=@n18, Fraction=@n19), B=VectorParameter() {parameterName=SSS Color Multiplier, defaultValue=<rgba 1 1 1 1>}.RGB)
   Alpha=Multiply(A=Multiply(A=@n4.R, B=@n20)) {constB=0.25}
  )
  n21 = Desaturation(None=@n17, Fraction=ScalarParameter() {parameterName=Specular Desaturation Amount})
  Multiply(A=@n4.R, B=@n21)
  Multiply(
   A=Multiply(A=@n21, B=TextureSample() {samplerType=SAMPLERTYPE_LinearColor, texture=&T_Noise01_Inverse}.R) {constB=0.2}
   B=ScalarParameter() {defaultValue=0.5, parameterName=Spec Value}
  ) {constB=0.5}
  Multiply()
  Multiply(
   A=MaterialFunctionCall!(In) {materialFunction=/Game/Packages/Fortress_SharedMaterials/Base_Material_Functions/Hologram}.Emissive
   B=Power(Base=@n4.R, Exp=ScalarParameter() {defaultValue=1, parameterName=Emissive_Mask})
  )
  n22 = Fresnel() {baseReflectFraction=0.4, exponent=2}
  OneMinus(None=Multiply(A=LinearInterpolate(Alpha=@n12.R) {constA=0.5}, B=Multiply(A=@n4.R, B=@n22)))
  ScalarParameter() {defaultValue=0.1, parameterName=Roughness}
  n23 = LinearInterpolate(A=Multiply(A=@n6.RGB) {constB=-1}, B=Constant3Vector() {constant=<rgba 0 0 1 0>}) {constAlpha=0.725}
  n24 = TextureSample() {samplerType=SAMPLERTYPE_Normal, texture=/Game/Creative/Sets/STW/Ingredients/Textures/CP_T_Resin_N}
  n25 = Constant3Vector() {constant=<rgba 1 1 0 0>}
  StaticSwitchParameter(True=Add(A=@n23, B=Multiply(A=@n24.RGB, B=@n25)), False=@n23) {parameterName=Use High poly normal map}
  TextureSample() {samplerType=SAMPLERTYPE_LinearColor, texture=/Game/Items/Art_noLOD/Textures/test}
 ]
}
